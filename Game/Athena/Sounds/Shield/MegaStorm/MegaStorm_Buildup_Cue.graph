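# Sound cue for the MegaStorm buildup, stored with the Athena shield sounds.
# A single wave player plays the MegaStorm_Buildup_01 wave with no randomization, mixing or modulation.
# It uses a map-wide attenuation asset with a very large max distance and routes through the Fort mission in-game sound class.
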SoundCue /Game/Athena/Sounds/Shield/MegaStorm/MegaStorm_Buildup_Cue
root=WavePlayer_0
nodes [{id=WavePlayer_0, properties {soundWaveAssetPtr=&/MegaStorm_Buildup_01}}]
properties {
 attenuationSettings=&/MegaStorm_Mapwide_Att
 maxDistance=2097152
 soundClassObject=/Game/Sounds/FortSoundClassesAndModes/Fort_Mission_IngameSounds
 volumeMultiplier=1
}
