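# Sound cue for the MegaStorm damage event in Athena, played with map-wide attenuation in the in-game mission sound class.
# A quality-level node at the root selects between two single-wave players.
# The first two quality slots share the standard wave and the last two share the "_opt" variant.
# The cue plays slightly boosted in volume, with no randomization or looping nodes.
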SoundCue /Game/Athena/Sounds/Shield/MegaStorm/MegaStorm_DamageEvent_Cue
root=QualityLevel_0
nodes [
 {id=QualityLevel_0, children [WavePlayer_0,WavePlayer_0,WavePlayer_1,WavePlayer_1]}
 {id=WavePlayer_0, properties {soundWaveAssetPtr=&/MegaStorm_DamageEvent_01}}
 {id=WavePlayer_1, properties {soundWaveAssetPtr=&/MegaStorm_DamageEvent_01_opt}}
]
properties {
 attenuationSettings=&/MegaStorm_Mapwide_Att
 maxDistance=2097152
 soundClassObject=/Game/Sounds/FortSoundClassesAndModes/Fort_Mission_IngameSounds
 volumeMultiplier=1.25
}
